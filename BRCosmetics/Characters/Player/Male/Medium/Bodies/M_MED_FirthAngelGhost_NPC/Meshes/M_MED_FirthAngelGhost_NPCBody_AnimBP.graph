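# Animation Blueprint for the Firth Angel Ghost NPC body mesh, layering secondary motion onto a pose copied from its attached parent mesh.
# The copied pose runs through the body's Control Rig, then chained Anim Dynamics nodes simulate the radio, knife and two necklace bones.
# Each simulated bone uses damped angular springs, reduced gravity and tight angular limits in root-relative space.
# The event graph holds only unconnected default nodes, with no variables or logic.
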AnimBlueprint /BRCosmetics/Characters/Player/Male/Medium/Bodies/M_MED_FirthAngelGhost_NPC/Meshes/M_MED_FirthAngelGhost_NPCBody_AnimBP
target_skeleton=&/M_MED_FirthAngelGhost_NPCBody_Skeleton
graphs [
 AnimGraph {nodes [
   n1 = Control Rig(Source=Copy Pose From Mesh() {bUseAttachedParent=true}) {
    alphaBoolBlend {blendOption=Linear}
    controlRigAssetReference {blueprintRigClass=&/M_MED_FirthAngelGhost_NPCBody_CtrlRig.M_MED_FirthAngelGhost_NPCBody_CtrlRig_C}
   }
   n2 = Anim Dynamics - Bone: dyn_radio_jnt(ComponentPose=Local To Component(LocalPose=@n1), Alpha=1, ExternalForce="0,0,0") {
    angularDampingOverride=1
    angularSpringConstant=30000
    bAngularSpring=true
    bOverrideAngularDamping=true
    bUseGravityOverride=true
    gravityScale=0.35
    simulationSpace=RootRelative
    alphaBoolBlend {blendOption=Linear}
    boundBone {boneName=dyn_radio_jnt}
    componentLinearAccScale=<1 1 1>
    componentLinearVelScale=<1 1 1>
    gravityOverride {z=-980}
    physicsBodyDefinitions [{
      boundBone {boneName=dyn_radio_jnt}
      boxExtents=<41 41 41>
      constraintSetup {angularLimitsMax=<7 1 5>, angularLimitsMin=<-7 -1 -5>, angularTarget=<1 0 0>}
      localJointOffset=<0 0 0>
     }]
   }
   n3 = Anim Dynamics - Bone: dyn_knife_jnt(ComponentPose=@n2, Alpha=1, ExternalForce="0,0,0") {
    angularDampingOverride=0.99
    angularSpringConstant=20000
    bAngularSpring=true
    bOverrideAngularDamping=true
    bUseGravityOverride=true
    gravityScale=0.35
    simulationSpace=RootRelative
    alphaBoolBlend {blendOption=Linear}
    boundBone {boneName=dyn_knife_jnt}
    componentLinearAccScale=<1 1 1>
    componentLinearVelScale=<1 1 1>
    gravityOverride {z=-980}
    physicsBodyDefinitions [{
      boundBone {boneName=dyn_knife_jnt}
      boxExtents=<40 40 40>
      constraintSetup {angularLimitsMax=<0.5 35 1>, angularLimitsMin=<-0.5 -30 -0.5>, angularTarget=<1 0 0>}
      localJointOffset=<0.04 0 0>
     }]
   }
   n4 = Anim Dynamics - Bone: dyn_necklace_01(ComponentPose=@n3, Alpha=1, ExternalForce="0,0,0") {
    angularDampingOverride=1
    angularSpringConstant=20000
    bAngularSpring=true
    bOverrideAngularDamping=true
    bUseGravityOverride=true
    gravityScale=0.35
    simulationSpace=RootRelative
    alphaBoolBlend {blendOption=Linear}
    boundBone {boneName=dyn_necklace_01}
    componentLinearAccScale=<1 1 1>
    componentLinearVelScale=<1 1 1>
    gravityOverride {z=-980}
    physicsBodyDefinitions [{
      boundBone {boneName=dyn_necklace_01}
      boxExtents=<45 45 45>
      constraintSetup {angularLimitsMax=<0 0 0>, angularLimitsMin=<0 -15 0>, angularTarget=<1 0 0>}
      localJointOffset=<4 0 0>
     }]
   }
   n5 = Anim Dynamics - Bone: dyn_necklace_02(ComponentPose=@n4, Alpha=1, ExternalForce="0,0,0") {
    angularDampingOverride=1
    angularSpringConstant=10000
    bAngularSpring=true
    bOverrideAngularDamping=true
    bUseGravityOverride=true
    gravityScale=0.35
    simulationSpace=RootRelative
    alphaBoolBlend {blendOption=Linear}
    boundBone {boneName=dyn_necklace_02}
    componentLinearAccScale=<1 1 1>
    componentLinearVelScale=<1 1 1>
    gravityOverride {z=-980}
    physicsBodyDefinitions [{
      boundBone {boneName=dyn_necklace_02}
      boxExtents=<30 30 30>
      constraintSetup {angularLimitsMax=<0 0 1>, angularLimitsMin=<0 -15 -1>, angularTarget=<1 0 0>}
      localJointOffset=<3 0 0>
     }]
   }
   Output Pose(Result=Component To Local(ComponentPose=@n5))
  ]}
 EventGraph {nodes [
   TryGetPawnOwner()
   Event BlueprintUpdateAnimation()
  ]}
]
variables []
